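# Translucent UI material that draws a soft-edged ring, named for the quick-time-event circle in the interrogation UI.
# An exponential radial gradient feeds two bands, each the difference of two SmoothSteps, combined with Max.
# Scalar parameters set the main band's position, width and softness through cumulative adds.
# The second, blur band reuses BlurEdge and BlurEdgeSofness, so its start equals its width and both edges share softness.
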Material /Game/Athena/UI/InGame/Interrogation/Materials/M_UI_QTECircle
graph {
 properties {
  bUsedWithStaticMesh=false
  blendMode=BLEND_Translucent
  materialDomain=MD_UI
  pixelDepthOffsetMode=PDOM_Legacy
  usedShadingModels=Default Lit
 }
 nodes [
  Constant() {r=1}
  n1 = MaterialFunctionCall() {materialFunction=&RadialGradientExponential}
  n2 = ScalarParameter() {parameterName=Edge}
  n3 = Add(A=@n2, B=ScalarParameter() {parameterName=EdgeSoftness})
  n4 = Add(A=@n3, B=ScalarParameter() {defaultValue=0.1, parameterName=InnerEdge})
  n5 = Add(A=@n4, B=ScalarParameter() {parameterName=InnerEdgeSoftness})
  n6 = ScalarParameter() {parameterName=BlurEdge}
  n7 = Add(A=@n6, B=ScalarParameter() {parameterName=BlurEdgeSofness})
  n8 = Add(A=@n7, B=ScalarParameter() {parameterName=BlurEdge})
  n9 = Add(A=@n8, B=ScalarParameter() {parameterName=BlurEdgeSofness})
  Max(
   A=Subtract(
    A=MaterialFunctionCall(Alpha=@n1.RadialGradientExponential, Min=@n2, Max=@n3) {materialFunction=&SmoothStep}.Result
    B=MaterialFunctionCall(Alpha=@n1.RadialGradientExponential, Min=@n4, Max=@n5) {materialFunction=&SmoothStep}.Result
   )
   B=Subtract(
    A=MaterialFunctionCall(Alpha=@n1.RadialGradientExponential, Min=@n6, Max=@n7) {materialFunction=&SmoothStep}.Result
    B=MaterialFunctionCall(Alpha=@n1.RadialGradientExponential, Min=@n8, Max=@n9) {materialFunction=&SmoothStep}.Result
   )
  )
 ]
}
