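# Additive UI material in the Interrogation UI folder that draws an animated, flickering radial starburst.
# A starburst texture, sampled in panning radial coordinates with switchable manual time, is multiplied by sine-powered noise and a smoothstepped radial falloff.
# Static switches add an outer radial mask and a scalable, invertible texture mask before the final alpha push.
# Color comes from either a curve-atlas gradient indexed by that falloff or a two-color blend.
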Material /Game/Athena/UI/InGame/Interrogation/Materials/M_UI_Starburst
graph {
 properties {bUsedWithStaticMesh=false, blendMode=BLEND_Additive, materialDomain=MD_UI, pixelDepthOffsetMode=PDOM_Legacy, usedShadingModels=Default Lit}
 nodes [
  n1 = VectorParameter() {group=Mask, parameterName=MaskScale, defaultValue=<rgba 0 0 0 1>}
  n2 = TextureCoordinate()
  n3 = MaterialFunctionCall(Float2=@n2) {materialFunction=&BreakOutFloat2Components}
  n4 = AppendVector(A=@n3.R, B=Multiply(A=@n3.G, B=ComponentMask(None=@n2) {g=true}))
  n5 = AppendVector(A=@n1.R, B=@n1.R)
  n6 = TextureSampleParameter2D!(
   UVs=Add(
    A=AppendVector(A=@n1.B, B=@n1.B)
    B=MaterialFunctionCall(UVs=@n4, Texture Scale=@n5) {materialFunction=&ScaleUVsByCenter}.UVs
   )
   Level
   Apply View MipBias
  ) {
   group=Mask
   mipValueMode=TMVM_MipLevel
   parameterName=MaskTexture
   samplerSource=SSM_Clamp_WorldGroupSettings
   texture=/Game/Athena/HUD/Vehicles/Materials/T_UI_ConeMask
  }
  n7 = LinearInterpolate(A=@n6.R, B=OneMinus(None=@n6.R), Alpha=ScalarParameter() {group=Mask, parameterName=InvertMask})
  n8 = VectorParameter() {parameterName=OuterRadial, defaultValue=<rgba 0.5 0.5 0.5 2.5>}
  n9 = AppendVector(A=@n8.R, B=@n8.R)
  n10 = ScalarParameter() {parameterName=FalloffEdge}
  n11 = MaterialFunctionCall(
   Alpha=MaterialFunctionCall(Density=ScalarParameter() {defaultValue=3, parameterName=FalloffDensity}) {materialFunction=&RadialGradientExponential}.RadialGradientExponential
   Min=@n10
   Max=Add(A=@n10, B=ScalarParameter() {defaultValue=0.2, parameterName=FallOffSoftness})
  ) {materialFunction=&SmoothStep}
  n12 = MaterialFunctionCall(Vector or UVs=Add(A=TextureCoordinate()) {constB=-0.5}) {materialFunction=&VectorToRadialValue}
  n13 = VectorParameter() {parameterName=Mult, defaultValue=<rgba 1 2 0 1>}
  n14 = AppendVector(A=@n13.R, B=@n13.R)
  n15 = Time()
  n16 = ScalarParameter() {parameterName=ManualTime}
  n17 = ScalarParameter() {parameterName=SpeedSwitch}
  n18 = VectorParameter() {parameterName=Speed, defaultValue=<rgba 0.1 -0.1 0 1>}
  n19 = MaterialFunctionCall(
   UVs=Panner(
    Coordinate=Multiply(A=@n12.Radial Coordinates, B=@n14)
    Time=LinearInterpolate(A=@n15, B=@n16, Alpha=@n17)
    Speed=AppendVector(A=@n18.R, B=@n18.R)
   ) {speedY=-0.2}
   Texture Scale=ScalarParameter() {defaultValue=1, parameterName=Scale}
  ) {materialFunction=&ScaleUVsByCenter}
  n20 = Multiply(A=TextureCoordinate(), B=ScalarParameter() {defaultValue=1, parameterName=NoiseScale})
  n21 = TextureSample(UVs=@n20) {samplerType=SAMPLERTYPE_Grayscale, texture=&LowResBlurredNoise}
  n22 = Time()
  n23 = ScalarParameter() {defaultValue=0.2, parameterName=FlickerSpeed}
  n24 = Multiply(
   A=ScalarParameter() {defaultValue=2, parameterName=FlickerPower}
   B=MaterialFunctionCall(Value=Multiply(A=@n22, B=@n23)) {materialFunction=&LinearSine}."Rounded Linear Sine "
  )
  n25 = Multiply(
   A=@n11.Result
   B=Multiply(
    A=TextureSampleParameter2D!(UVs=@n19.UVs, Apply View MipBias) {parameterName=StarTexture, texture=/Game/Global/Textures/Default/PointSplatter/T_UI_StarBurst_Scatter}.G
    B=Power(Base=@n21.RGB, Exp=@n24)
   )
  )
  n26 = StaticSwitchParameter(
   True=Multiply(
    A=@n25
    B=MaterialFunctionCall(CenterPosition=@n9, Radius=@n8.B, Density=@n8.B) {materialFunction=&RadialGradientExponential}.RadialGradientExponential
   )
   False=@n25
  ) {parameterName=OuterRadialMask}
  Multiply(
   A=StaticSwitchParameter(True=Multiply(A=@n26, B=@n7), False=@n26) {group=Mask, parameterName=UseMask}
   B=ScalarParameter() {defaultValue=1, parameterName=FinalAlphaPush}
  )
  n27 = MaterialFunctionCall(Vector or UVs=Add(A=TextureCoordinate()) {constB=-0.5}) {materialFunction=&VectorToRadialValue}
  n28 = VectorParameter() {parameterName=Mult, defaultValue=<rgba 1 2 0 1>}
  n29 = AppendVector(A=@n28.R, B=@n28.R)
  Panner(Coordinate=Multiply(A=Multiply(A=@n27.Radial Coordinates, B=@n29), B=@n20)) {speedY=-0.2}
  n30 = CurveAtlasRowParameter(CurveTime=@n11.Result) {
   atlas=/Game/Athena/HUD/Materials/Victory/NewCurveLinearColorAtlas
   curve=/Game/Athena/Playlists/Goose/UI/Material/GooseVictoryGradient1
   defaultValue=9
   parameterName=CurveParams
  }
  StaticSwitchParameter(
   True=@n30
   False=LinearInterpolate(
    A=VectorParameter() {parameterName=Color2, defaultValue=<rgba 0.1 0 1 1>}.RGB
    B=VectorParameter() {parameterName=Color1, defaultValue=<rgba 0 1 1 1>}.RGB
   )
  ) {parameterName=UseColorCurve}
 ]
}
